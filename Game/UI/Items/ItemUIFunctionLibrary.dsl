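; ItemUIFunctionLibrary is a Blueprint function library of helpers for Fortnite item UI.
; It maps item tier enums to integers and treats an invalid object or one of class FortNullItem as a null item.
; It also parses the number between "Level " and a colon in a string.
; TruncateIntegerValue abbreviates large integers into localized K, M or B text with caller-set fractional digits.
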(fn ConvertTierToInteger
  (params (Tier :type EFortItemTier%20Enum :default "No_Tier"))
  (return (Utilities|Select :Index Tier :No_Tier 0 :I 1 :II 2 :III 3 :IV 4 :V 5 :VI 6 :VII 7 :VIII 8 :IX 9 :X 10)))

(fn IsNullItem
  (params (ItemObject :type Object%20Reference))
  (Utilities|IsValid ItemObject
    (:Is%20Valid
      (return (Utilities|Equal(Class) (Utilities|GetClass ItemObject) "/Script/FortniteUI.FortNullItem")))
    (:Is%20Not%20Valid
      (return true))))

(fn ParseLevelRequiredFromString
  (params (inString :type String))
  (return (Utilities|String|StringToInteger (Utilities|String|GetSubstring inString (+ (Utilities|String|FindSubstring inString "Level ") 6) (+ (- (- (Utilities|String|FindSubstring inString ":") 1) (+ (Utilities|String|FindSubstring inString "Level ") 6)) 1)))))

(fn TruncateIntegerValue
  (params (Value :type Integer) (Min%20Fractional%20Digits :type Integer) (Max%20Fractional%20Digits :type Integer))
  (return (Utilities|Text|FormatText "NSLOCTEXT(\"\", \"BD42172042885927315950971FBADB72\", \"{num}{symbol}\")" (Utilities|Text|ToText(Float) (select (>= Value 1000000000) (* (Math|Conversions|ToFloat(Integer) Value) 1e-09) (select (>= Value 1000000) (* (Math|Conversions|ToFloat(Integer) Value) 1e-06) (select (>= Value 1000) (* (Math|Conversions|ToFloat(Integer) Value) 0.001) (Math|Conversions|ToFloat(Integer) Value)))) "HalfToEven" false true 1 3 Min%20Fractional%20Digits Max%20Fractional%20Digits) (select (>= Value 1000000000) "NSLOCTEXT(\"\", \"8E61599D40482E4A82BCFEBA69D36426\", \"B\")" (select (>= Value 1000000) "NSLOCTEXT(\"\", \"FF102F0343E475960EFC76956D09A582\", \"M\")" (select (>= Value 1000) "NSLOCTEXT(\"\", \"687DB60A46A2845F244207989D76A0D4\", \"K\")" ""))))))
