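# Translucent UI material that draws an item cooldown meter as a radial progress ring with configurable fill and background colors.
# Rotated, optionally mirrored UVs become an angle, which is added to the amount and floored for a hard-edged sweep, with optional normalization against the offset.
# Two subtracted sphere gradients form the ring from radius and hardness parameters.
# The masked sweep blends the colors, while background alpha keeps the unfilled ring visible.
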Material /Game/UI/Items/Materials/Item-CooldownMeter_Material
graph {
 properties {
  bUsedWithStaticMesh=false
  blendMode=BLEND_Translucent
  materialDomain=MD_UI
  pixelDepthOffsetMode=PDOM_Legacy
  shadingModel=MSM_Unlit
  usedShadingModels=Unlit
 }
 nodes [
  n1 = ScalarParameter() {defaultValue=1, parameterName=Amount_0-1}
  n2 = ScalarParameter() {parameterName=Offset_0-1}
  n3 = LinearInterpolate(A=@n1, B=Multiply(A=@n1, B=OneMinus(None=@n2)), Alpha=ScalarParameter() {parameterName=bNormalizeAmount}) {desc="If normalized, amount reduced by (1 - offset)"}
  n4 = Clamp(None=@n2) {desc="Removes gap when Amount  = 1.0", minDefault=0.001}
  n5 = ScalarParameter() {defaultValue=0.25, parameterName=Origin_0-1}
  n6 = LinearInterpolate(Alpha=ScalarParameter() {parameterName=bInvertX}) {constA=1, constB=-1}
  n7 = LinearInterpolate(Alpha=ScalarParameter() {parameterName=bInvertY}) {constA=1, constB=-1}
  n8 = MaterialFunctionCall(UVs=TextureCoordinate(), "Rotation Angle (0-1)"=Multiply(A=Add(A=@n4, B=@n5), B=Multiply(A=@n6, B=@n7))) {materialFunction=&CustomRotator}
  n9 = Constant2Vector() {g=2, r=2}
  n10 = AppendVector(A=@n6, B=@n7)
  n11 = MaterialFunctionCall(Vector or UVs=Multiply(A=Subtract(A=@n8.Rotated Values) {constB=0.5}, B=Multiply(A=@n9, B=@n10)) {constB=2}) {materialFunction=&VectorToRadialValue}
  n12 = ScalarParameter() {defaultValue=1, parameterName=OuterRadius_0-1}
  n13 = MaterialFunctionCall!(UVs, CenterPosition, Radius=Clamp(None=Multiply(A=@n12) {constB=0.5}) {maxDefault=0.5}) {materialFunction=/Engine/Functions/Engine_MaterialFunctions01/Gradient/SphereGradient-2D}
  n14 = ScalarParameter() {defaultValue=3.2, parameterName=Hardness}
  n15 = ScalarParameter() {defaultValue=0.8, parameterName=InnerRadius_0-1}
  n16 = MaterialFunctionCall(Radius=Clamp(None=Multiply(A=@n15) {constB=0.5}) {maxDefault=0.495}) {materialFunction=&SphereGradient-2D}
  n17 = Multiply(A=@n15, B=@n14)
  n18 = Subtract(A=Clamp(None=Multiply(A=@n13.Result 0-1, B=@n14) {constB=3}), B=Clamp(None=Multiply(A=@n16.Result 0-1, B=@n17) {constB=1.8}))
  n19 = Multiply(A=Floor(None=Add(A=@n3, B=@n11.Vector Converted to Angle)), B=@n18)
  LinearInterpolate(
   A=VectorParameter() {parameterName=BGColor, defaultValue=<rgba 0.0052 0.0052 0.0056 0>}.RGB
   B=VectorParameter() {parameterName=Color, defaultValue=<rgba 1 1 1 1>}.RGB
   Alpha=@n19
  )
  Max(A=Multiply(A=@n18, B=VectorParameter() {parameterName=BGColor, defaultValue=<rgba 0.0052 0.0052 0.0056 0>}.A), B=@n19)
 ]
}
