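# Translucent unlit UI material that draws a segmented radial pie or ring fill driven by an amount parameter, with a drop shadow.
# Rotated, invertible UVs become a radial angle whose floored sum with the amount, optionally normalized against the rotation offset, masks an outer-minus-inner sphere-gradient ring cut by a deprecated radial segments function.
# The mask is rebuilt at offset UVs for the shadow, and a DebugFloat4Values readout of the rotation offset feeds color and opacity.
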Material /Game/UI/Items/Materials/M-PieMenu
graph {
 properties {
  bUsedWithStaticMesh=false
  blendMode=BLEND_Translucent
  materialDomain=MD_UI
  numCustomizedUVs=1
  pixelDepthOffsetMode=PDOM_Legacy
  shadingModel=MSM_Unlit
  usedShadingModels=Unlit
 }
 nodes [
  n1 = VectorParameter() {parameterName=ShadowColor, defaultValue=<rgba 0.0052 0.0052 0.0056 0>}
  n2 = ScalarParameter() {defaultValue=0.01, group=SegmentSpacing, parameterName="EdgeSmooth (S)"}
  n3 = ScalarParameter() {group=SegmentSpacing, parameterName=Segments}
  n4 = ScalarParameter() {group=SegmentSpacing, parameterName="Rotation (S)"}
  n5 = ScalarParameter() {defaultValue=0.9, group=SegmentSpacing, parameterName=SegmentWidth}
  n6 = ScalarParameter() {defaultValue=1, parameterName=Amount_0-1}
  n7 = Abs(None=Divide(A=ScalarParameter() {parameterName=Rotation}, B=ScalarParameter() {defaultValue=1, parameterName=NormalizeOffsetValue}))
  n8 = LinearInterpolate(A=@n6, B=Multiply(A=@n6, B=OneMinus(None=@n7)), Alpha=ScalarParameter() {parameterName=bNormalizeAmount}) {desc="If normalized, amount reduced by (1 - offset)"}
  n9 = LinearInterpolate(Alpha=ScalarParameter() {parameterName=bInvertX}) {constA=1, constB=-1}
  n10 = LinearInterpolate(Alpha=ScalarParameter() {parameterName=bInvertY}) {constA=1, constB=-1}
  n11 = Multiply(
   A=Add(
    A=Clamp(None=@n7) {desc="Removes gap when Amount  = 1.0", minDefault=0.001}
    B=ScalarParameter() {defaultValue=0.25, parameterName=Origin_0-1}
   )
   B=Multiply(A=@n9, B=@n10)
  )
  n12 = Subtract(A=MaterialFunctionCall(UVs=TextureCoordinate(), "Rotation Angle (0-1)"=@n11) {materialFunction=&CustomRotator}.Rotated Values) {constB=0.5}
  n13 = AppendVector(A=@n9, B=@n10)
  n14 = Multiply(A=Constant2Vector() {g=2, r=2}, B=@n13)
  n15 = Add(
   A=@n8
   B=MaterialFunctionCall(Vector or UVs=Multiply(A=@n12, B=@n14) {constB=2}) {materialFunction=&VectorToRadialValue}.Vector Converted to Angle
  )
  n16 = TextureCoordinate()
  n17 = Clamp(None=Multiply(A=ScalarParameter() {defaultValue=1, parameterName=OuterRadius_0-1}) {constB=0.5}) {maxDefault=0.5}
  n18 = ScalarParameter() {defaultValue=3.2, parameterName=Hardness}
  n19 = Clamp(None=Multiply(
    A=MaterialFunctionCall!(UVs=@n16, CenterPosition, Radius=@n17) {materialFunction=/Engine/Functions/Engine_MaterialFunctions01/Gradient/SphereGradient-2D}.Result 0-1
    B=@n18
   ) {constB=3})
  n20 = ScalarParameter() {defaultValue=0.8, parameterName=InnerRadius_0-1}
  n21 = Clamp(None=Multiply(A=@n20) {constB=0.5}) {maxDefault=0.495}
  n22 = Multiply(A=@n20, B=@n18)
  n23 = Clamp(None=Multiply(
    A=MaterialFunctionCall(UVs=@n16, Radius=@n21) {materialFunction=&SphereGradient-2D}.Result 0-1
    B=@n22
   ) {constB=1.8})
  n24 = Multiply(
   A=OneMinus(None=MaterialFunctionCall!(EdgeSmooth=@n2, Repeat=@n3, CenterOffset, Rotation=@n4, UV, SegmentSpacing=@n5) {materialFunction=/Game/Athena/HUD/reticles/MF_UI_RadialSegments_DEPRECATED}.Result)
   B=Multiply(A=Floor(None=@n15), B=Subtract(A=@n19, B=@n23))
  )
  n25 = VectorParameter() {parameterName="debugTexture Scale (V2)", defaultValue=<rgba 1 1 0 0>}
  n26 = TextureCoordinate()
  n27 = VectorParameter() {parameterName=debug, defaultValue=<rgba 0.3 0.3 0.7 0.7>}
  n28 = VectorParameter() {parameterName=debugspacing, defaultValue=<rgba 0 0.35 1 0.65>}
  n29 = MaterialFunctionCall!(
   Vector4=Divide(A=ScalarParameter() {parameterName=Rotation}, B=ScalarParameter() {defaultValue=1, parameterName=NormalizeOffsetValue})
   MaximumNumberOfDigits=Constant() {r=3}
   UVs=Add(
    A=AppendVector(A=@n25.B, B=@n25.B)
    B=MaterialFunctionCall(UVs=@n26, Texture Scale=@n25.RGB) {materialFunction=&ScaleUVsByCenter}.UVs
   )
   DebugTextLocation RG_UpperRight BA_LowerLeft=MaterialFunctionCall(R=@n27.R, G=@n27.R, B=@n27.R, A=@n27.R) {materialFunction=&AppendMany}.RGBA
   Component Spacing=MaterialFunctionCall(R=@n28.R, G=@n28.R, B=@n28.R, A=@n28.R) {materialFunction=&AppendMany}.RGBA
  ) {materialFunction=/Engine/Functions/Engine_MaterialFunctions02/Utility/DebugFloat4Values}
  n30 = MaterialFunctionCall(Float4=@n29.ColorCodedOutput) {materialFunction=&BreakOutFloat4Components}
  n31 = MaterialFunctionCall(R=@n30.R, G=@n30.R, B=@n30.R) {materialFunction=&AppendMany}
  LinearInterpolate(
   A=LinearInterpolate(A=@n1.RGB, B=VectorParameter() {parameterName=Color, defaultValue=<rgba 1 1 1 1>}.RGB, Alpha=@n24)
   B=@n31.RGB
   Alpha=@n31.RGB
  )
  n32 = VectorParameter() {parameterName=GlobalScaleOffset, defaultValue=<rgba 1 1 0.5 0.5>}
  MaterialFunctionCall!(
   Texture Scale=AppendVector(A=@n32.R, B=@n32.R)
   UVs=TextureCoordinate()
   Offset XY Texturesize Z=MaterialFunctionCall(R=@n32.B, G=@n32.B, B=Constant() {r=1}) {materialFunction=&AppendMany}.RGB
  ) {materialFunction=/Engine/Functions/Engine_MaterialFunctions01/Texturing/ScaleUVsAroundPoint}
  n33 = ScalarParameter() {defaultValue=0.01, group=SegmentSpacing, parameterName="EdgeSmooth (S)"}
  n34 = ScalarParameter() {group=SegmentSpacing, parameterName=Segments}
  n35 = VectorParameter() {parameterName=ShadowOffset, defaultValue=<rgba -0.02 -0.02 0 1>}
  n36 = AppendVector(A=@n35.R, B=@n35.R)
  n37 = ScalarParameter() {group=SegmentSpacing, parameterName="Rotation (S)"}
  n38 = ScalarParameter() {defaultValue=0.9, group=SegmentSpacing, parameterName=SegmentWidth}
  n39 = VectorParameter() {parameterName=ShadowOffset, defaultValue=<rgba -0.02 -0.02 0 1>}
  n40 = Add(A=TextureCoordinate(), B=AppendVector(A=@n39.R, B=@n39.R))
  n41 = Subtract(A=MaterialFunctionCall(UVs=@n40, "Rotation Angle (0-1)"=@n11) {materialFunction=&CustomRotator}.Rotated Values) {constB=0.5}
  n42 = Multiply(A=Constant2Vector() {g=2, r=2}, B=@n13)
  n43 = Add(
   A=@n8
   B=MaterialFunctionCall(Vector or UVs=Multiply(A=@n41, B=@n42) {constB=2}) {materialFunction=&VectorToRadialValue}.Vector Converted to Angle
  )
  n44 = Clamp(None=Multiply(
    A=MaterialFunctionCall(UVs=@n40, Radius=@n17) {materialFunction=&SphereGradient-2D}.Result 0-1
    B=@n18
   ) {constB=3})
  n45 = Clamp(None=Multiply(
    A=MaterialFunctionCall(UVs=@n40, Radius=@n21) {materialFunction=&SphereGradient-2D}.Result 0-1
    B=@n22
   ) {constB=1.8})
  n46 = Multiply(
   A=OneMinus(None=MaterialFunctionCall(EdgeSmooth=@n33, Repeat=@n34, CenterOffset=@n36, Rotation=@n37, SegmentSpacing=@n38) {materialFunction=&MF_UI_RadialSegments_DEPRECATED}.Result)
   B=Multiply(A=Floor(None=@n43), B=Subtract(A=@n44, B=@n45))
  )
  Max(A=Multiply(A=@n29.GreyScaleOutput), B=Max(A=Multiply(A=@n1.A, B=@n46), B=@n24))
 ]
}
