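# Translucent UI material that draws a pie-menu ring with an adjustable radial fill and segment gaps, with Color and Final_Alpha parameters.
# Two sphere gradients cut the inner and outer radii, multiplied by a fill mask built from the sine of a rotated, invertible radial angle against Amount.
# A second sine of angle times the floored segment count carves separators, tapered by a radial gradient, with hardness multipliers controlling edge sharpness.
# The amount-normalization and UV-scaling branches feed no other node.
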Material /Game/UI/Items/Materials/M-PieMenu_SmoothSine
graph {
 properties {
  bUsedWithStaticMesh=false
  blendMode=BLEND_Translucent
  materialDomain=MD_UI
  numCustomizedUVs=1
  pixelDepthOffsetMode=PDOM_Legacy
  shadingModel=MSM_Unlit
  usedShadingModels=Unlit
 }
 nodes [
  n1 = ScalarParameter() {defaultValue=1, parameterName=Amount_0-1}
  n2 = Abs(None=Divide(A=ScalarParameter() {parameterName=Rotation}, B=ScalarParameter() {defaultValue=1, parameterName=NormalizeOffsetValue}))
  LinearInterpolate(A=@n1, B=Multiply(A=@n1, B=OneMinus(None=@n2)), Alpha=ScalarParameter() {parameterName=bNormalizeAmount}) {desc="If normalized, amount reduced by (1 - offset)"}
  n3 = VectorParameter() {parameterName=GlobalScaleOffset, defaultValue=<rgba 1 1 0.5 0.5>}
  n4 = ScalarParameter() {parameterName=PostRotation}
  MaterialFunctionCall!(
   Texture Scale=AppendVector(A=@n3.R, B=@n3.R)
   UVs=MaterialFunctionCall(UVs=TextureCoordinate(), "Rotation Angle (0-1)"=Multiply(A=@n4) {constB=0.01}) {materialFunction=&CustomRotator}.Rotated Values
   Offset XY Texturesize Z=MaterialFunctionCall(R=@n3.B, G=@n3.B, B=Constant() {r=1}) {materialFunction=&AppendMany}.RGB
  ) {materialFunction=/Engine/Functions/Engine_MaterialFunctions01/Texturing/ScaleUVsAroundPoint}
  n5 = Add(
   A=Clamp(None=@n2) {desc="Removes gap when Amount  = 1.0", minDefault=0.001}
   B=ScalarParameter() {defaultValue=0.25, parameterName=Origin_0-1}
  )
  n6 = LinearInterpolate(Alpha=ScalarParameter() {parameterName=bInvertX}) {constA=1, constB=-1}
  n7 = LinearInterpolate(Alpha=ScalarParameter() {parameterName=bInvertY}) {constA=1, constB=-1}
  n8 = Multiply(A=@n6, B=@n7)
  n9 = Subtract(A=MaterialFunctionCall(UVs=TextureCoordinate(), "Rotation Angle (0-1)"=Multiply(A=@n5, B=@n8)) {materialFunction=&CustomRotator}.Rotated Values) {constB=0.5}
  n10 = Multiply(A=Constant2Vector() {g=2, r=2}, B=AppendVector(A=@n6, B=@n7))
  n11 = Sine(None=MaterialFunctionCall(Vector or UVs=Multiply(A=@n9, B=@n10) {constB=2}) {materialFunction=&VectorToRadialValue}.Vector Converted to Angle) {period=2}
  n12 = Multiply(A=ScalarParameter() {defaultValue=1, parameterName=Amount_0-1}) {constB=1.5}
  n13 = OneMinus(None=Multiply(A=Subtract(A=@n11, B=@n12), B=ScalarParameter() {defaultValue=256, parameterName=FillnoSegmentEdgeHardness}))
  n14 = TextureCoordinate()
  n15 = ScalarParameter() {defaultValue=1, parameterName=OuterRadius_0-1}
  n16 = MaterialFunctionCall!(UVs=@n14, CenterPosition, Radius=Clamp(None=Multiply(A=@n15) {constB=0.5}) {maxDefault=0.5}) {materialFunction=/Engine/Functions/Engine_MaterialFunctions01/Gradient/SphereGradient-2D}
  n17 = ScalarParameter() {defaultValue=3.2, parameterName=HardnessOuterRadius}
  n18 = Multiply(A=ScalarParameter() {defaultValue=0.8, parameterName=InnerRadius_0-1}) {constB=0.5}
  n19 = Multiply(
   A=MaterialFunctionCall(UVs=@n14, Radius=Clamp(None=@n18) {maxDefault=0.495}) {materialFunction=&SphereGradient-2D}.Result 0-1
   B=ScalarParameter() {defaultValue=3.2, parameterName=HardnessInnerRadius}
  ) {constB=1.8}
  n20 = DotProduct(A=Clamp(None=Multiply(A=@n16.Result 0-1, B=@n17) {constB=3}), B=OneMinus(None=Clamp(None=@n19)))
  n21 = Saturate(None=DotProduct(A=Saturate(None=@n13), B=Saturate(None=@n20)))
  n22 = MaterialFunctionCall(UVs=TextureCoordinate(), "Rotation Angle (0-1)"=ScalarParameter() {parameterName="Rotation (S)"}) {materialFunction=&CustomRotator}
  n23 = Constant2Vector() {g=-0.5, r=-0.5}
  n24 = Multiply(
   A=MaterialFunctionCall(Vector or UVs=Add(A=@n22.Rotated Values, B=@n23)) {materialFunction=&VectorToRadialValue}.Vector Converted to Angle
   B=Floor(None=ScalarParameter() {defaultValue=5, parameterName=NumberOfSegments})
  )
  n25 = Subtract(A=OneMinus(None=Sine(None=@n24)), B=ScalarParameter() {parameterName=SegmentGap})
  n26 = ScalarParameter() {defaultValue=128, parameterName=SegmentEdgeHardness}
  n27 = OneMinus(None=MaterialFunctionCall() {materialFunction=&RadialGradientExponential}.RadialGradientExponential)
  n28 = ScalarParameter() {defaultValue=0.75, parameterName=TaperSegmentLineMultiplier}
  n29 = LinearInterpolate(A=ScalarParameter() {parameterName=LiftSegmentLineMultiplier}, Alpha=Saturate(None=Power(Base=@n27, Exp=@n28)))
  n30 = Saturate(None=Multiply(A=Saturate(None=@n25), B=Multiply(A=@n26, B=@n29)))
  Multiply(A=Saturate(None=DotProduct(A=@n21, B=@n30)), B=ScalarParameter() {defaultValue=1, parameterName=Final_Alpha})
  VectorParameter() {parameterName=Color, defaultValue=<rgba 1 1 1 1>}
 ]
}
