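# Translucent UI material that draws a soft-edged circle with a colored outline around an inner fill.
# A center-scaled UV feeds an exponential radial gradient, which two SmoothStep thresholds share, each softened by a Softness parameter.
# The threshold at OuterCircle plus OutlineSize blends OutlineColor into InnerColor, while the one at OuterCircle forms the outer circle mask.
# Scale, edge thresholds, softness and both colors are exposed as parameters.
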Material /Game/UI/Items/Materials/M_UI_OutlineCircle
graph {
 properties {
  bUsedWithStaticMesh=false
  blendMode=BLEND_Translucent
  materialDomain=MD_UI
  pixelDepthOffsetMode=PDOM_Legacy
  usedShadingModels=Default Lit
 }
 nodes [
  n1 = MaterialFunctionCall(UVs=MaterialFunctionCall(Texture Scale=ScalarParameter() {defaultValue=1, parameterName=Scale}) {materialFunction=&ScaleUVsByCenter}.UVs) {materialFunction=&RadialGradientExponential}
  n2 = ScalarParameter() {defaultValue=0.1, parameterName=OuterCircle}
  n3 = Add(A=@n2, B=ScalarParameter() {defaultValue=0.1, parameterName=OutlineSize})
  n4 = ScalarParameter() {defaultValue=0.02, parameterName=Softness}
  LinearInterpolate(
   A=VectorParameter() {parameterName=OutlineColor, defaultValue=<rgba 0.0677 0.0677 0.0677 1>}.RGB
   B=VectorParameter() {parameterName=InnerColor, defaultValue=<rgba 1 1 1 1>}.RGB
   Alpha=MaterialFunctionCall(Alpha=@n1.RadialGradientExponential, Min=Subtract(A=@n3, B=@n4), Max=@n3) {materialFunction=&SmoothStep}.Result
  )
  MaterialFunctionCall(Alpha=@n1.RadialGradientExponential, Min=Subtract(A=@n2, B=@n4), Max=@n2) {materialFunction=&SmoothStep}
 ]
}
